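# Editor light-review report that flags point light components with questionable Detail Mode settings.
# Its RunComponentLogic override reads the warning mode from the light review database and calls the parent, returning success immediately when the parent returns true.
# Otherwise it casts to PointLightComponent, flagging an overridden detailMode property or a value equal to the warning mode.
# Each finding adds an error and raises severity; success means no errors.
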EditorUtilityBlueprint /Game/NeverCook/ArtTools/Tech_Art/LightReviewTool/Report_LightReview_DetailMode
graphs [
 EventGraph {links [], nodes []}
 RunComponentLogic {nodes [
   n1 = RunComponentLogic!() -> (then, Comp:Light Component Object Reference) {is=title K2Node_FunctionEntry}
   > n2 = Set LComp(LComp:Light Component Object Reference=@n1) -> (Output_Get:Light Component Object Reference)
   > Set DetailModeToWarning(DetailModeToWarning:EDetailMode Enum=Get DetailModeToWarning!(self:BP Light Review Tool Database Object Reference=Get LightDatabase() -> (LightDatabase:BP Light Review Tool Database Object Reference)) -> (DetailModeToWarning:EDetailMode Enum=DM_Low)) -> (Output_Get:EDetailMode Enum=DM_Low)
   > n3 = Parent: RunComponentLogic!(execute, Comp:Light Component Object Reference=@n2) -> (then, Success?:bool, Error:Array of Strings, severity:Enum_GART_ReportSeverity Enum) {is=title K2Node_CallParentFunction}
   > n4 = Branch(Condition=@n3)
   > Return Node!(execute, Success?:bool=true, Error:Array of Strings, severity:Enum_GART_ReportSeverity Enum=NewEnumerator0)
   n5 = Sequence!(execute=@n4.else) -> (then_0, then_1, then_2:Exec, then_3:Exec)
   > n6 = Cast To PointLightComponent!(execute, Object=Get LComp() -> (LComp:Light Component Object Reference)) -> (then, CastFailed, AsPoint Light Component:Point Light Component Object Reference) {is=title K2Node_DynamicCast}
   > n7 = IsEditorPropertyOverridden!(execute, Object:Object Reference=@n6, PropertyName:Name=detailMode) -> (then, ReturnValue:EEditorPropertyValueState Enum=Default)
   > Set LPropertyOverriden(LPropertyOverriden:bool="Equal (Enum)"(A:EEditorPropertyValueState Enum=@n7, B:EEditorPropertyValueState Enum=Overridden)) -> (Output_Get:bool)
   > Set LDetailMode(LDetailMode:EDetailMode Enum=Get DetailMode!(self:Scene Component Object Reference=@n6) -> (DetailMode:EDetailMode Enum=DM_Low)) -> (Output_Get:EDetailMode Enum=DM_Low)
   Branch(execute=@n5.then_1, Condition=Get LPropertyOverriden() -> (LPropertyOverriden:bool))
   > Add(TargetArray=Get l_errors() -> (l_errors:Array of Strings), NewItem=The Detail Mode is not with the original value.) {is=title K2Node_CallArrayFunction}
   > IncreaseErrorSeverity!(
    execute
    self:GART Report Base Object Reference
    Severity:"Enum_GART_ReportSeverity Enum (by ref)"=Get l_severity()
    NewSeverity:Enum_GART_ReportSeverity Enum=NewEnumerator1
   ) -> (then)
   n8 = Get LDetailMode() -> (LDetailMode:EDetailMode Enum=DM_Low)
   Branch(
    execute=@n5.then_2
    Condition="Equal (Enum)"(A:EDetailMode Enum=@n8, B:EDetailMode Enum=Get DetailModeToWarning!() -> (DetailModeToWarning=DM_Low))
   )
   > Add(
    TargetArray=Get l_errors()
    NewItem=Append!(A="The Detail Mode is set to ", B=Enum to String!(Enumerator:EDetailMode Enum=@n8) -> (ReturnValue:String), C:String=.) -> (ReturnValue)
   )
   > IncreaseErrorSeverity(Severity=Get l_severity(), NewSeverity=NewEnumerator1)
   n9 = Get l_errors()
   Return Node(
    execute=@n5.then_3
    Success?=NOT Boolean(A=Is Valid Index!(TargetArray=@n9, IndexToTest:int) -> (ReturnValue:bool))
    Error=Get l_errors()
    severity=Get l_severity()
   )
  ]}
]
variables [DetailModeToWarning {type=EDetailMode, default=DM_Low}]
