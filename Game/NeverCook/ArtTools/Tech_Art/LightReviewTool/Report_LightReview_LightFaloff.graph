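# Editor utility report for the Light Review tool that, by its name, checks light components for falloff.
# Its RunComponentLogic override reads the max light falloff from the Light Review Tool database into the inherited MaxFalloff variable.
# It then calls the parent implementation on the light component and passes its success, errors and severity straight through, so the check itself lives in the parent class.
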EditorUtilityBlueprint /Game/NeverCook/ArtTools/Tech_Art/LightReviewTool/Report_LightReview_LightFaloff
graphs [
 EventGraph {links [], nodes []}
 RunComponentLogic {nodes [
   n1 = RunComponentLogic!() -> (then, Comp:Light Component Object Reference) {is=title K2Node_FunctionEntry}
   > Set MaxFalloff(MaxFalloff:float=Get MaxLightFalloff!(self:BP Light Review Tool Database Object Reference=Get LightDatabase() -> (LightDatabase:BP Light Review Tool Database Object Reference)) -> (MaxLightFalloff:double)) -> (Output_Get:float)
   > n2 = Parent: RunComponentLogic!(execute, Comp:Light Component Object Reference=@n1) -> (then, Success?:bool, Error:Array of Strings, severity:Enum_GART_ReportSeverity Enum) {is=title K2Node_CallParentFunction}
   > Return Node!(execute, Success?:bool=@n2, Error:Array of Strings=@n2.Error, severity:Enum_GART_ReportSeverity Enum=@n2.severity)
  ]}
]
variables []
