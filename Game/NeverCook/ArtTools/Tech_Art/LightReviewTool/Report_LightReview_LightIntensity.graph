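# Editor utility report for the Light Review Tool that flags light components whose intensity exceeds a maximum.
# It reads the limit from the light database, then runs the parent check, returning success immediately if that returns true.
# Otherwise it reads intensity only after a successful PointLightComponent cast, compares it to the limit, and logs an error with raised severity.
# It returns success only when the error list is empty, along with errors and severity.
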EditorUtilityBlueprint /Game/NeverCook/ArtTools/Tech_Art/LightReviewTool/Report_LightReview_LightIntensity
graphs [
 EventGraph {links [], nodes []}
 RunComponentLogic {nodes [
   n1 = RunComponentLogic!() -> (then, Comp:Light Component Object Reference) {is=title K2Node_FunctionEntry}
   > n2 = Set LComp(LComp:Light Component Object Reference=@n1) -> (Output_Get:Light Component Object Reference)
   > Set MaxIntensity(MaxIntensity:double=Get MaxLightIntensity!(self:BP Light Review Tool Database Object Reference=Get LightDatabase() -> (LightDatabase:BP Light Review Tool Database Object Reference)) -> (MaxLightIntensity:double)) -> (Output_Get:double)
   > n3 = Parent: RunComponentLogic!(execute, Comp:Light Component Object Reference=@n2) -> (then, Success?:bool, Error:Array of Strings, severity:Enum_GART_ReportSeverity Enum) {is=title K2Node_CallParentFunction}
   > n4 = Branch(Condition=@n3)
   > Return Node!(execute, Success?:bool=true, Error:Array of Strings, severity:Enum_GART_ReportSeverity Enum=NewEnumerator0)
   n5 = Sequence!(execute=@n4.else) -> (then_0, then_1, then_2:Exec)
   > n6 = Cast To PointLightComponent!(execute, Object=Get LComp() -> (LComp:Light Component Object Reference)) -> (then, CastFailed, AsPoint Light Component:Point Light Component Object Reference) {is=title K2Node_DynamicCast}
   > Set LIntensity(LIntensity:float=Get Intensity!(self:Light Component Base Object Reference=@n6) -> (Intensity:float)) -> (Output_Get:float)
   n7 = Get MaxIntensity() -> (MaxIntensity:double)
   n8 = Get LIntensity() -> (LIntensity:float)
   Branch(execute=@n5.then_1, Condition="float > float"(A=@n8, B=@n7))
   > Add(
    TargetArray=Get l_errors() -> (l_errors:Array of Strings)
    NewItem=Append!(
     A="Has a Intensity of "
     B="To String (Float)"(InDouble=@n8)
     C:String=" which should be less or equal than "
     D:String="To String (Float)"(InDouble=@n7)
     E:String=.
    ) -> (ReturnValue)
   ) {is=title K2Node_CallArrayFunction}
   > IncreaseErrorSeverity!(
    execute
    self:GART Report Base Object Reference
    Severity:"Enum_GART_ReportSeverity Enum (by ref)"=Get l_severity()
    NewSeverity:Enum_GART_ReportSeverity Enum=NewEnumerator2
   ) -> (then)
   n9 = Get l_errors()
   Return Node(
    execute=@n5.then_2
    Success?=NOT Boolean(A=Is Valid Index!(TargetArray=@n9, IndexToTest:int) -> (ReturnValue:bool))
    Error=Get l_errors()
    severity=Get l_severity()
   )
  ]}
]
variables [MaxIntensity {type=number, default=300}]
