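# Light review report (editor utility) that flags spot lights whose outer or inner cone angle exceeds the maximum set in the light review tool database.
# RunComponentLogic reads both limits, calls the parent check, and returns success immediately if that returns true.
# Otherwise it casts to a spot light, compares both angles, appends an error and raises severity for each overrun.
# It then returns the error list and severity, succeeding only when no errors exist.
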EditorUtilityBlueprint /Game/NeverCook/ArtTools/Tech_Art/LightReviewTool/Report_LightReview_LightOuterAndInnerCone
graphs [
 EventGraph {links [], nodes []}
 RunComponentLogic {nodes [
   n1 = RunComponentLogic!() -> (then, Comp:Light Component Object Reference) {is=title K2Node_FunctionEntry}
   > n2 = Set LComp(LComp:Light Component Object Reference=@n1) -> (Output_Get:Light Component Object Reference)
   n3 = Get LightDatabase() -> (LightDatabase:BP Light Review Tool Database Object Reference)
   > Set OutMaxAngle(OutMaxAngle:double=Get OutMaxConeAngle!(self:BP Light Review Tool Database Object Reference=@n3) -> (OutMaxConeAngle:double)) -> (Output_Get:double)
   > Set InnerMaxAngle(InnerMaxAngle:double=Get InMaxConeAngle!(self:BP Light Review Tool Database Object Reference=@n3) -> (InMaxConeAngle:double)) -> (Output_Get:double)
   > n4 = Parent: RunComponentLogic!(execute, Comp:Light Component Object Reference=@n2) -> (then, Success?:bool, Error:Array of Strings, severity:Enum_GART_ReportSeverity Enum) {is=title K2Node_CallParentFunction}
   > n5 = Branch(Condition=@n4)
   > Return Node!(execute, Success?:bool=true, Error:Array of Strings, severity:Enum_GART_ReportSeverity Enum=NewEnumerator0)
   n6 = Sequence!(execute=@n5.else) -> (then_0, then_1, then_2:Exec, then_3:Exec)
   > n7 = Cast To SpotLightComponent!(execute, Object=Get LComp() -> (LComp:Light Component Object Reference)) -> (then, CastFailed, AsSpot Light Component:Spot Light Component Object Reference) {is=title K2Node_DynamicCast}
   > Set LOuterConeAngle(LOuterConeAngle:float=Get OuterConeAngle!(self:Spot Light Component Object Reference=@n7) -> (OuterConeAngle:float)) -> (Output_Get:float)
   > Set LInnerConeAngle(LInnerConeAngle:float=Get InnerConeAngle!(self:Spot Light Component Object Reference=@n7) -> (InnerConeAngle:float)) -> (Output_Get:float)
   n8 = Get LOuterConeAngle() -> (LOuterConeAngle:float)
   n9 = Get OutMaxAngle() -> (OutMaxAngle:double)
   Branch(execute=@n6.then_1, Condition="float > float"(A=@n8, B=@n9))
   > Add(
    TargetArray=Get l_errors() -> (l_errors:Array of Strings)
    NewItem=Append!(
     A="Has a  Outer Cone Angle of "
     B="To String (Float)"(InDouble=@n8)
     C:String=" which should be less or equal than "
     D:String="To String (Float)"(InDouble=@n9)
     E:String=.
    ) -> (ReturnValue)
   ) {is=title K2Node_CallArrayFunction}
   > IncreaseErrorSeverity!(
    execute
    self:GART Report Base Object Reference
    Severity:"Enum_GART_ReportSeverity Enum (by ref)"=Get l_severity()
    NewSeverity:Enum_GART_ReportSeverity Enum=NewEnumerator2
   ) -> (then)
   n10 = Get LInnerConeAngle() -> (LInnerConeAngle:float)
   n11 = Get InnerMaxAngle() -> (InnerMaxAngle:double)
   Branch(execute=@n6.then_2, Condition="float > float"(A=@n10, B=@n11))
   > Add(
    TargetArray=Get l_errors()
    NewItem=Append(
     A="Has a  Inner Cone Angle of "
     B="To String (Float)"(InDouble=@n10)
     C=" which should be less or equal than "
     D="To String (Float)"(InDouble=@n11)
     E=.
    )
   )
   > IncreaseErrorSeverity(Severity=Get l_severity(), NewSeverity=NewEnumerator2)
   n12 = Get l_errors()
   Return Node(
    execute=@n6.then_3
    Success?=NOT Boolean(A=Is Valid Index!(TargetArray=@n12, IndexToTest:int) -> (ReturnValue:bool))
    Error=Get l_errors()
    severity=Get l_severity()
   )
  ]}
]
variables [
 InnerMaxAngle {type=number, default=80}
 OutMaxAngle {type=number, default=80}
]
