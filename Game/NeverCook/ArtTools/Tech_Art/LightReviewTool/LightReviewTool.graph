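# Editor utility widget that reviews lights for inconsistencies from a single "Search for inconsistency" button.
# On click, inline Python reloads LightValidatorUtils and calls GetAllLightActors to gather the objects to check.
# These run through seven configurable GART report classes covering intensity, detail mode, indirect lighting, attenuation, falloff, cone angles and volumetric scattering.
# The resulting report map feeds a newly spawned tab of the generic EUW_GART_List_Main review widget.
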EditorUtilityWidgetBlueprint /Game/NeverCook/ArtTools/Tech_Art/LightReviewTool/LightReviewTool
graphs [EventGraph {nodes [
   "On Clicked (EditorUtilityButton_32)"() {is=title K2Node_ComponentBoundEvent}
   > n1 = GetSelectionSet(self=Get LevelEditorSubsystem())
   > n2 = Execute Python Script!(
    execute
    PythonScript="import sys\r\npath = project_content_dir+\"/Python/TechArt/Tools\"\r\nsys.path.append(project_content_dir+\"&Tools/\")\r\nimport LightValidatorUtils\r\nimport importlib\r\nimportlib.reload(LightValidatorUtils)\r\ncomponents_filtered = LightValidatorUtils.GetAllLightActors()"
    project_content_dir:String=ProjectContentDir!() -> (ReturnValue:String)
   ) -> (then, ReturnValue, components_filtered:Array of Object References) {inputs [project_content_dir], outputs [components_filtered]}
   > n3 = RunObjectsThroughFilters!(
    execute
    Objects:Array of Object References=@n2.components_filtered
    ReportClasses:Array of GART Report Base Soft Class References=Get Report Classes() -> (Report Classes:Array of GART Report Base Soft Class References)
    __WorldContext:Object Reference=@n1
   ) -> (then, ReportDataMap:Map of Object References to Struct GART Asset Report List Structures)
   > n4 = SpawnAndRegisterTabAndGetID!(
    execute
    self:Editor Utility Subsystem Object Reference=Get EditorUtilitySubsystem()
    InBlueprint:Editor Utility Widget Blueprint Object Reference=/Game/NeverCook/EditorUtilities/GenericAssetReviewTool/Widgets/EUW_GART_List_Main
   ) -> (then, NewTabID:Name=None, ReturnValue:Editor Utility Widget Object Reference)
   > n5 = Cast To EUW_GART_List_Main!(execute, Object=@n4.ReturnValue) -> (then, CastFailed, AsEUW GART List Main:EUW GART List Main Object Reference) {is=title K2Node_DynamicCast}
   > SetObjectReportData!(
    execute
    self:EUW GART List Main Object Reference=@n5
    ReportData:Map of Object References to Struct GART Asset Report List Structures=@n3
    UpdateList:bool=true
   ) -> (then)
  ]}]
variables [Report Classes {
  type="array<Class>"
  default [
   &/Report_LightReview_LightIntensity.Report_LightReview_LightIntensity_C
   &/Report_LightReview_DetailMode.Report_LightReview_DetailMode_C
   &/Report_LightReview_IndirectLighting.Report_LightReview_IndirectLighting_C
   &/Report_LightReview_LightAttenuation.Report_LightReview_LightAttenuation_C
   &/Report_LightReview_LightFaloff.Report_LightReview_LightFaloff_C
   &/Report_LightReview_LightOuterAndInnerCone.Report_LightReview_LightOuterAndInnerCone_C
   &/Report_LightReview_VolumetricScattering.Report_LightReview_VolumetricScattering_C
  ]
 }]
widgets {
 info {rootWidgetClass=&Overlay, widgetCount=7}
 widgets [
  Overlay_0 {widgetClassPath=&Overlay, widget=Overlay_0}
  VerticalBox_125 {
   widgetClassPath=&VerticalBox
   parent=Overlay_0
   widget=VerticalBox_125
   slot {class=OverlaySlot, properties {horizontalAlignment=HAlign_Fill, verticalAlignment=VAlign_Fill}}
  }
  HorizontalBox_0 {
   widgetClassPath=&HorizontalBox
   parent=VerticalBox_125
   widget=HorizontalBox_0
   slot {class=VerticalBoxSlot, properties {size {sizeRule=Fill, value=1}}}
  }
  SizeBox_0 {
   widgetClassPath=&SizeBox
   parent=HorizontalBox_0
   widget=SizeBox_0
   properties {bOverride_MinDesiredHeight=true, heightOverride=129.8934, maxAspectRatio=64.8307, minAspectRatio=-55.601, minDesiredHeight=95}
   slot {class=HorizontalBoxSlot, properties {verticalAlignment=VAlign_Center, padding=<ltrb 25 25 25 25>, size {sizeRule=Fill, value=1}}}
  }
  EditorUtilityButton_32 {widgetClassPath=&EditorUtilityButton, parent=SizeBox_0, widget=EditorUtilityButton_32, slot {class=SizeBoxSlot}}
  ScaleBox_0 {
   widgetClassPath=&ScaleBox
   parent=EditorUtilityButton_32
   widget=ScaleBox_0
   slot {class=ButtonSlot, properties {horizontalAlignment=HAlign_Fill, verticalAlignment=VAlign_Fill, padding=<ltrb 15 15 15 15>}}
  }
  TextBlock_60 {
   widgetClassPath=&TextBlock
   parent=ScaleBox_0
   widget=TextBlock_60
   properties {
    text=Search for inconsistency
    font {
     bForceMonospaced=false
     bMaterialIsStencil=false
     fontMaterial=null
     fontObject=&Roboto
     letterSpacing=0
     monospacedWidth=1
     size=12
     skewAmount=0
     typefaceFontName=Bold
     outlineSettings {
      bApplyOutlineToDropShadows=false
      bMiteredCorners=false
      bSeparateFillAlpha=false
      outlineMaterial=null
      outlineSize=0
      outlineColor=<rgba 0 0 0 1>
     }
    }
   }
   slot {class=ScaleBoxSlot}
  }
 ]
}
